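# Sound cue for an indoor ambient spot sound of a building's piped appliance running a start/stop cycle, in the domestic and commercial POI audio set.
# A Random node picks one of three equally weighted wave variations, and a Modulator slightly varies its volume.
# It uses a shared wide one-shot tile-ambient attenuation and the indoor ambient sound class.
# A concurrency override limits it to two simultaneous instances, and the debug flag is left on.
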SoundCue /WildEstate/Assets/Audio/POI/Domestic_Commercial/Spots/SC_Spot_Building_Integrated_Piped_Appliance_Start_Stop_Cycle
root=Modulator_0
nodes [
 {id=Modulator_0, children [Random_1], properties {volumeMax=1, volumeMin=0.9}}
 {id=Random_1, children [WavePlayer_0,WavePlayer_1,WavePlayer_4], properties {weights [1,1,1]}}
 {
  id=WavePlayer_0
  properties {soundWaveAssetPtr=&/Spot_Building_Integrated_Piped_Appliance_Start_Stop_Cycle_01}
 }
 {
  id=WavePlayer_1
  properties {soundWaveAssetPtr=&/Spot_Building_Integrated_Piped_Appliance_Start_Stop_Cycle_02}
 }
 {
  id=WavePlayer_4
  properties {soundWaveAssetPtr=&/Spot_Building_Integrated_Piped_Appliance_Start_Stop_Cycle_03}
 }
]
properties {
 attenuationSettings=/Game/Sounds/Attenuation/Fort_TileAmbient_OneShot_Wider_Att
 bDebug=true
 bOverrideConcurrency=true
 maxDistance=8000
 soundClassObject=&Fort_Ambient_Indoors_Athena
 volumeMultiplier=1
 attenuationOverrides {bSpatialize=false, stereoSpread=1147.9967, attenuationShapeExtents {x=6826.1992}}
 concurrencyOverrides {maxCount=2}
 concurrencySet [/AmbientAudioBase/Concurrency/AmbientConc_2Max_Loop_PreventNew_Default_Concurrency]
}
