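# Sound Cue for a one-shot building ambience spot: a pipe fitting releasing a sharp air burst, classed as indoor ambient sound.
# A Random node picks among eight equally weighted wave variations and feeds a Modulator that slightly varies volume.
# It uses a shared wide one-shot ambient attenuation asset and overrides concurrency to cap simultaneous instances at two.
# The cue's debug flag is left on.
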SoundCue /WildEstate/Assets/Audio/POI/Domestic_Commercial/Spots/SC_Spot_Building_Pipe_Fitting_Releasing_Sharp_Air_Burst
root=Modulator_0
nodes [
 {id=Modulator_0, children [Random_1], properties {volumeMax=1, volumeMin=0.9}}
 {
  id=Random_1
  children [WavePlayer_0, WavePlayer_1, WavePlayer_4, WavePlayer_6, WavePlayer_7, WavePlayer_8, WavePlayer_9, WavePlayer_10]
  properties {weights [1,1,1,1,1,1,1,1]}
 }
 {
  id=WavePlayer_0
  properties {soundWaveAssetPtr=&/Spot_Building_Pipe_Fitting_Releasing_Sharp_Air_Burst_08}
 }
 {
  id=WavePlayer_1
  properties {soundWaveAssetPtr=&/Spot_Building_Pipe_Fitting_Releasing_Sharp_Air_Burst_07}
 }
 {
  id=WavePlayer_10
  properties {soundWaveAssetPtr=&/Spot_Building_Pipe_Fitting_Releasing_Sharp_Air_Burst_01}
 }
 {
  id=WavePlayer_4
  properties {soundWaveAssetPtr=&/Spot_Building_Pipe_Fitting_Releasing_Sharp_Air_Burst_06}
 }
 {
  id=WavePlayer_6
  properties {soundWaveAssetPtr=&/Spot_Building_Pipe_Fitting_Releasing_Sharp_Air_Burst_05}
 }
 {
  id=WavePlayer_7
  properties {soundWaveAssetPtr=&/Spot_Building_Pipe_Fitting_Releasing_Sharp_Air_Burst_04}
 }
 {
  id=WavePlayer_8
  properties {soundWaveAssetPtr=&/Spot_Building_Pipe_Fitting_Releasing_Sharp_Air_Burst_03}
 }
 {
  id=WavePlayer_9
  properties {soundWaveAssetPtr=&/Spot_Building_Pipe_Fitting_Releasing_Sharp_Air_Burst_02}
 }
]
properties {
 attenuationSettings=/Game/Sounds/Attenuation/Fort_TileAmbient_OneShot_Wider_Att
 bDebug=true
 bOverrideConcurrency=true
 maxDistance=8000
 soundClassObject=&Fort_Ambient_Indoors_Athena
 volumeMultiplier=1
 attenuationOverrides {bSpatialize=false, stereoSpread=1147.9967, attenuationShapeExtents {x=6826.1992}}
 concurrencyOverrides {maxCount=2}
 concurrencySet [/AmbientAudioBase/Concurrency/AmbientConc_2Max_Loop_PreventNew_Default_Concurrency]
}
