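# Sound cue for subtle building stress creak spot ambience in domestic and commercial POIs, routed to the indoor ambient sound class.
# A Random node picks one of ten equally weighted creak waves and feeds a Modulator that slightly varies volume.
# A shared one-shot attenuation asset shapes falloff, and overridden concurrency caps it at two voices.
# It sends to default reverb and, with linear distance-based level, distant reverb; its debug flag is on.
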SoundCue /WildEstate/Assets/Audio/POI/Domestic_Commercial/Spots/SC_Spot_Building_Subtle_Stress_Creaks
root=Modulator_0
nodes [
 {id=Modulator_0, children [Random_1], properties {volumeMax=1, volumeMin=0.9}}
 {
  id=Random_1
  children [
   WavePlayer_0
   WavePlayer_1
   WavePlayer_4
   WavePlayer_13
   WavePlayer_14
   WavePlayer_15
   WavePlayer_16
   WavePlayer_17
   WavePlayer_18
   WavePlayer_19
  ]
  properties {weights [1,1,1,1,1,1,1,1,1,1]}
 }
 {id=WavePlayer_0, properties {soundWaveAssetPtr=&/Spot_Building_Subtle_Stress_Creaks_10}}
 {id=WavePlayer_1, properties {soundWaveAssetPtr=&/Spot_Building_Subtle_Stress_Creaks_09}}
 {
  id=WavePlayer_13
  properties {soundWaveAssetPtr=&/Spot_Building_Subtle_Stress_Creaks_07}
 }
 {
  id=WavePlayer_14
  properties {soundWaveAssetPtr=&/Spot_Building_Subtle_Stress_Creaks_06}
 }
 {
  id=WavePlayer_15
  properties {soundWaveAssetPtr=&/Spot_Building_Subtle_Stress_Creaks_05}
 }
 {
  id=WavePlayer_16
  properties {soundWaveAssetPtr=&/Spot_Building_Subtle_Stress_Creaks_04}
 }
 {
  id=WavePlayer_17
  properties {soundWaveAssetPtr=&/Spot_Building_Subtle_Stress_Creaks_03}
 }
 {
  id=WavePlayer_18
  properties {soundWaveAssetPtr=&/Spot_Building_Subtle_Stress_Creaks_02}
 }
 {
  id=WavePlayer_19
  properties {soundWaveAssetPtr=&/Spot_Building_Subtle_Stress_Creaks_01}
 }
 {id=WavePlayer_4, properties {soundWaveAssetPtr=&/Spot_Building_Subtle_Stress_Creaks_08}}
]
properties {
 attenuationSettings=/Game/Sounds/Attenuation/Fort_TileAmbient_OneShot_Wider_Att
 bDebug=true
 bOverrideConcurrency=true
 maxDistance=8000
 soundClassObject=&Fort_Ambient_Indoors_Athena
 volumeMultiplier=1
 attenuationOverrides {bSpatialize=false, stereoSpread=1147.9967, attenuationShapeExtents {x=6826.1992}}
 concurrencyOverrides {maxCount=2}
 concurrencySet [/AmbientAudioBase/Concurrency/AmbientConc_2Max_Loop_PreventNew_Default_Concurrency]
 soundSubmixSends [
  {minSendLevel=0, soundSubmix=&Fortnite_Reverb_Default}
  {
   maxSendDistance=6000
   maxSendLevel=0.45
   minSendDistance=3000
   minSendLevel=0.15
   sendLevel=0.45
   sendLevelControlMethod=Linear
   soundSubmix=&Fortnite_Reverb_Distant
  }
 ]
}
