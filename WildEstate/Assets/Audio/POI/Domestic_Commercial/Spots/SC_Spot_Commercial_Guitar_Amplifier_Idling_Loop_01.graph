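# Indoor ambient spot cue for an idling guitar amplifier in WildEstate commercial interiors, using very-small attenuation.
# A mixer layers repeated random interference one-shots, spaced by random delays, over a looping hum that randomly plays either directly or delayed, pitched up and faded in.
# Only the first two quality-level slots carry interference; the other two are empty.
# An outer modulator and looping envelope add slight volume variation.
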SoundCue /WildEstate/Assets/Audio/POI/Domestic_Commercial/Spots/SC_Spot_Commercial_Guitar_Amplifier_Idling_Loop_01
root=Enveloper_0
nodes [
 {id=Delay_0, children [QualityLevel_0], properties {delayMax=10, delayMin=5}}
 {id=Delay_1, children [Modulator_1], properties {delayMax=0.5, delayMin=0.5}}
 {
  id=Enveloper_0
  children [Modulator_0]
  properties {
   bLoop=true
   bLoopIndefinitely=true
   loopEnd=20
   pitchCurve {editorCurveData {keys [{time=0}]}}
   volumeCurve {editorCurveData {keys [<at -0.0028 1>,<at 0 1>,<at 10 0.95>,<at 20 1>]}}
  }
 }
 {
  id=Enveloper_1
  children [WavePlayer_12]
  properties {
   bLoopIndefinitely=true
   loopEnd=20
   pitchCurve {editorCurveData {keys [{time=0}]}}
   volumeCurve {editorCurveData {keys [<at -0.0028 0>,<at 0 0>,<at 1 1>,<at 20 1>]}}
  }
 }
 {id=Looping_0, children [Delay_0]}
 {id=Mixer_0, children [Looping_0,Random_0], properties {inputVolume [0.6,0.33]}}
 {id=Modulator_0, children [Mixer_0], properties {volumeMax=1, volumeMin=0.9}}
 {id=Modulator_1, children [Enveloper_1], properties {pitchMax=1.1, pitchMin=1.1}}
 {id=QualityLevel_0, children [Random_1,Random_1,null,null]}
 {id=Random_0, children [Delay_1, WavePlayer_12], properties {weights [1,1]}}
 {
  id=Random_1
  children [WavePlayer_0, WavePlayer_1, WavePlayer_4, WavePlayer_5, WavePlayer_6, WavePlayer_7, WavePlayer_8, WavePlayer_9, WavePlayer_10, WavePlayer_11]
  properties {weights [1,1,1,1,1,1,1,1,1,1]}
 }
 {
  id=WavePlayer_0
  properties {soundWaveAssetPtr=&/Spot_Commercial_Guitar_Amplifier_Idling_Interference_Oneshot_02}
 }
 {
  id=WavePlayer_1
  properties {soundWaveAssetPtr=&/Spot_Commercial_Guitar_Amplifier_Idling_Interference_Oneshot_03}
 }
 {
  id=WavePlayer_10
  properties {soundWaveAssetPtr=&/Spot_Commercial_Guitar_Amplifier_Idling_Interference_Oneshot_10}
 }
 {
  id=WavePlayer_11
  properties {soundWaveAssetPtr=&/Spot_Commercial_Guitar_Amplifier_Idling_Interference_Oneshot_01}
 }
 {
  id=WavePlayer_12
  properties {bLooping=true, soundWaveAssetPtr=&/Spot_Commercial_Guitar_Amplifier_Idling_Loop_01}
 }
 {
  id=WavePlayer_4
  properties {soundWaveAssetPtr=&/Spot_Commercial_Guitar_Amplifier_Idling_Interference_Oneshot_04}
 }
 {
  id=WavePlayer_5
  properties {soundWaveAssetPtr=&/Spot_Commercial_Guitar_Amplifier_Idling_Interference_Oneshot_05}
 }
 {
  id=WavePlayer_6
  properties {soundWaveAssetPtr=&/Spot_Commercial_Guitar_Amplifier_Idling_Interference_Oneshot_06}
 }
 {
  id=WavePlayer_7
  properties {soundWaveAssetPtr=&/Spot_Commercial_Guitar_Amplifier_Idling_Interference_Oneshot_07}
 }
 {
  id=WavePlayer_8
  properties {soundWaveAssetPtr=&/Spot_Commercial_Guitar_Amplifier_Idling_Interference_Oneshot_08}
 }
 {
  id=WavePlayer_9
  properties {soundWaveAssetPtr=&/Spot_Commercial_Guitar_Amplifier_Idling_Interference_Oneshot_09}
 }
]
properties {
 attenuationSettings=&/Spot_WildEstate_Domestic_Commercial_VerySmall_Att
 bDebug=true
 bOverrideConcurrency=true
 maxDistance=700
 soundClassObject=&Fort_Ambient_Indoors_Athena
 volumeMultiplier=1
 attenuationOverrides {bSpatialize=false, stereoSpread=1147.9967, attenuationShapeExtents {x=6826.1992}}
 concurrencyOverrides {maxCount=2}
 concurrencySet [/AmbientAudioBase/Concurrency/AmbientConc_1Max_Prop_StopLowestPriority_Concurrency]
}
