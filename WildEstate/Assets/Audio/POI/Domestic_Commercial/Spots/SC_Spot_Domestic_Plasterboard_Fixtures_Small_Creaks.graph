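# Sound Cue for ambient indoor spot one-shots of small plasterboard and fixture creaks at domestic/commercial POIs.
# A Random node picks among ten equally weighted creak waves, then a Modulator slightly varies volume.
# It uses a shared wide one-shot ambient attenuation asset, with overridden concurrency capping it at two instances.
# Submix sends feed default reverb and a distant reverb whose level scales linearly with distance.
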SoundCue /WildEstate/Assets/Audio/POI/Domestic_Commercial/Spots/SC_Spot_Domestic_Plasterboard_Fixtures_Small_Creaks
root=Modulator_0
nodes [
 {id=Modulator_0, children [Random_0], properties {volumeMax=1, volumeMin=0.9}}
 {
  id=Random_0
  children [
   WavePlayer_2
   WavePlayer_3
   WavePlayer_5
   WavePlayer_6
   WavePlayer_7
   WavePlayer_8
   WavePlayer_9
   WavePlayer_10
   WavePlayer_11
   WavePlayer_12
  ]
  properties {weights [1,1,1,1,1,1,1,1,1,1]}
 }
 {
  id=WavePlayer_10
  properties {soundWaveAssetPtr=&/Spot_Domestic_Plasterboard_Fixtures_Small_Creaks_03}
 }
 {
  id=WavePlayer_11
  properties {soundWaveAssetPtr=&/Spot_Domestic_Plasterboard_Fixtures_Small_Creaks_02}
 }
 {
  id=WavePlayer_12
  properties {soundWaveAssetPtr=&/Spot_Domestic_Plasterboard_Fixtures_Small_Creaks_01}
 }
 {
  id=WavePlayer_2
  properties {soundWaveAssetPtr=&/Spot_Domestic_Plasterboard_Fixtures_Small_Creaks_10}
 }
 {
  id=WavePlayer_3
  properties {soundWaveAssetPtr=&/Spot_Domestic_Plasterboard_Fixtures_Small_Creaks_09}
 }
 {
  id=WavePlayer_5
  properties {soundWaveAssetPtr=&/Spot_Domestic_Plasterboard_Fixtures_Small_Creaks_08}
 }
 {
  id=WavePlayer_6
  properties {soundWaveAssetPtr=&/Spot_Domestic_Plasterboard_Fixtures_Small_Creaks_07}
 }
 {
  id=WavePlayer_7
  properties {soundWaveAssetPtr=&/Spot_Domestic_Plasterboard_Fixtures_Small_Creaks_06}
 }
 {
  id=WavePlayer_8
  properties {soundWaveAssetPtr=&/Spot_Domestic_Plasterboard_Fixtures_Small_Creaks_05}
 }
 {
  id=WavePlayer_9
  properties {soundWaveAssetPtr=&/Spot_Domestic_Plasterboard_Fixtures_Small_Creaks_04}
 }
]
properties {
 attenuationSettings=/Game/Sounds/Attenuation/Fort_TileAmbient_OneShot_Wider_Att
 bDebug=true
 bOverrideConcurrency=true
 maxDistance=8000
 soundClassObject=&Fort_Ambient_Indoors_Athena
 volumeMultiplier=1
 attenuationOverrides {bSpatialize=false, stereoSpread=1147.9967, attenuationShapeExtents {x=6826.1992}}
 concurrencyOverrides {maxCount=2}
 concurrencySet [/AmbientAudioBase/Concurrency/AmbientConc_2Max_Loop_PreventNew_Default_Concurrency]
 soundSubmixSends [
  {minSendLevel=0, soundSubmix=&Fortnite_Reverb_Default}
  {
   maxSendDistance=6000
   maxSendLevel=0.6
   minSendDistance=3000
   minSendLevel=0.1
   sendLevel=0.45
   sendLevelControlMethod=Linear
   soundSubmix=&Fortnite_Reverb_Distant
  }
 ]
}
