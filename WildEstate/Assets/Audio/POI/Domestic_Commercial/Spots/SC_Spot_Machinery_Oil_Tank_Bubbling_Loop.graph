SoundCue /WildEstate/Assets/Audio/POI/Domestic_Commercial/Spots/SC_Spot_Machinery_Oil_Tank_Bubbling_Loop
root=Random_1
nodes [
 {id=Random_1, children [WavePlayer_0,WavePlayer_1,WavePlayer_4], properties {weights [1,1,1]}}
 {
  id=WavePlayer_0
  properties {bLooping=true, soundWaveAssetPtr=&/Spot_Machinery_Oil_Tank_Bubbling_Loop_01}
 }
 {
  id=WavePlayer_1
  properties {bLooping=true, soundWaveAssetPtr=&/Spot_Machinery_Oil_Tank_Bubbling_Loop_02}
 }
 {
  id=WavePlayer_4
  properties {bLooping=true, soundWaveAssetPtr=&/Spot_Machinery_Oil_Tank_Bubbling_Loop_03}
 }
]
properties {
 attenuationSettings=/WildEstate/Assets/Audio/POI/SlurpFactory/Spots/Spot_WildEstate_SlurpFactory_Vats_Att
 bDebug=true
 bOverrideConcurrency=true
 maxDistance=2500
 soundClassObject=&Fort_Ambient_Indoors_Athena
 volumeMultiplier=1
 attenuationOverrides {bSpatialize=false, stereoSpread=1147.9967, attenuationShapeExtents {x=6826.1992}}
 concurrencyOverrides {maxCount=3}
 concurrencySet [/AmbientAudioBase/Concurrency/AmbientConc_2Max_Loop_PreventNew_Default_Concurrency]
}
