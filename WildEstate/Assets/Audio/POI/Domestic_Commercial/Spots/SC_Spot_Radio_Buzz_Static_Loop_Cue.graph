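# Looping spot ambience cue for a radio buzzing with static, classed as an ambient prop in domestic/commercial POIs.
# Mixes three looping waves (glitchy radio comms pitched well down, vintage computer idle, TV static) at weighted volumes with randomized pitch modulation.
# An inline attenuation node at the root adds HRTF spatialization, occlusion, focus weighting and distance filtering.
# Overridden concurrency caps simultaneous instances, and a submix send feeds the default reverb.
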SoundCue /WildEstate/Assets/Audio/POI/Domestic_Commercial/Spots/SC_Spot_Radio_Buzz_Static_Loop_Cue
root=Attenuation_5
nodes [
 {
  id=Attenuation_5
  children [Modulator_0]
  properties {
   bOverrideAttenuation=true
   attenuationOverrides {
    bEnableFocusInterpolation=true
    bEnableLogFrequencyScaling=true
    bEnableOcclusion=true
    coneSphereFalloffDistance=250
    coneSphereRadius=50
    distanceAlgorithm=Logarithmic
    falloffDistance=1000
    focusAzimuth=15
    focusDistanceScale=0.9
    focusReleaseInterpSpeed=0.8
    focusVolumeAttenuation=1.2
    hPFFrequencyAtMax=100
    hPFFrequencyAtMin=50
    lPFFrequencyAtMax=750
    lPFRadiusMax=1500
    lPFRadiusMin=300
    nonFocusAzimuth=45
    nonFocusVolumeAttenuation=0.2
    occlusionInterpolationTime=2
    occlusionLowPassFilterFrequency=10000
    occlusionTraceChannel=ECC_GameTraceChannel11
    occlusionVolumeAttenuation=0.25
    spatializationAlgorithm=SPATIALIZATION_HRTF
    stereoSpread=0
    attenuationShapeExtents=<250 61.3579 22>
    customAttenuationCurve {editorCurveData {keys [
       {arriveTangent=-2.6039, interpMode=RCIM_Cubic, leaveTangent=-2.6039, tangentMode=RCTM_User, time=0, value=1}
       {arriveTangent=0.0377, interpMode=RCIM_Cubic, leaveTangent=0.0377, tangentMode=RCTM_User, time=0.5, value=0}
      ]}}
   }
  }
 }
 {id=Mixer_0, children [Modulator_1,Modulator_3,Modulator_4], properties {inputVolume [0.5,1.5,1.25]}}
 {id=Modulator_0, children [Mixer_0], properties {pitchMin=1}}
 {id=Modulator_1, children [WavePlayer_3], properties {pitchMax=0.6, pitchMin=0.55}}
 {id=Modulator_3, children [WavePlayer_2], properties {pitchMax=1.1}}
 {id=Modulator_4, children [WavePlayer_1]}
 {
  id=WavePlayer_1
  properties {bLooping=true, soundWaveAssetPtr=/WildEstate/Assets/Audio/POI/Manor/Spots/Spot_WildEstate_Manor_Huge_TV_Static_Loop}
 }
 {
  id=WavePlayer_2
  properties {bLooping=true, soundWaveAssetPtr=&/Spot_Vintage_Computer_Idling_Loop_03}
 }
 {
  id=WavePlayer_3
  properties {bLooping=true, soundWaveAssetPtr=&/Spot_Radio_Comms_Glitchy_Loop}
 }
]
properties {
 bOverrideConcurrency=true
 maxDistance=1250
 soundClassObject=&Fort_Ambient_Props_Athena
 volumeMultiplier=0.3
 attenuationOverrides {falloffDistance=2000}
 concurrencyOverrides {maxCount=3}
 concurrencySet [&Weapon_Drop_StopOldest]
 soundSubmixSends [{minSendLevel=0.25, sendLevelControlMethod=Linear, soundSubmix=&Fortnite_Reverb_Default}]
}
