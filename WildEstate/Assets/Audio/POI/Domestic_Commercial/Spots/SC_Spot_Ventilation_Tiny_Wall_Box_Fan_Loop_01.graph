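# Looping ambient spot sound cue for a tiny wall box fan ventilation prop in WildEstate's domestic and commercial POIs.
# A looping slow desk-fan wave runs through a Modulator that randomly pitches it down, then into a Mixer whose second input is empty.
# It plays quietly on the ambient props sound class with very-small-radius attenuation.
# An overridden concurrency limit caps simultaneous instances at a low count, despite an assigned prop concurrency set.
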SoundCue /WildEstate/Assets/Audio/POI/Domestic_Commercial/Spots/SC_Spot_Ventilation_Tiny_Wall_Box_Fan_Loop_01
root=Mixer_1
nodes [
 {id=Mixer_1, children [Modulator_0,null], properties {inputVolume [1,2]}}
 {id=Modulator_0, children [WavePlayer_0], properties {pitchMax=0.8, pitchMin=0.7, volumeMax=1}}
 {id=WavePlayer_0, properties {bLooping=true, soundWaveAssetPtr=/AmbientAudioContent/Waves/SpottedLoops/Spot_Desk_Fan_Slow_Loop}}
]
properties {
 attenuationSettings=&/Spot_WildEstate_Domestic_Commercial_VerySmall_Att
 bEnableBusSends=false
 bOverrideConcurrency=true
 maxDistance=700
 soundClassObject=&Fort_Ambient_Props_Athena
 volumeMultiplier=0.35
 concurrencyOverrides {maxCount=2, voiceStealReleaseTime=2}
 concurrencySet [/Game/Sounds/Concurrency_Rules/Athena_Prop]
 preEffectBusSends [{}]
}
