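# Sound cue for an indoor ambient spot loop of a vintage desktop computer idling.
# A random node picks one of seven level-trimmed looping waves, then a modulator slightly varies pitch and volume.
# An endlessly looping enveloper adds a subtle volume dip, feeding only the first two quality levels; the other two are empty.
# Overrides supply short-range HRTF attenuation with occlusion and listener focus, plus a tight voice limit.
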SoundCue /WildEstate/Assets/Audio/POI/Domestic_Commercial/Spots/SC_Spot_Vintage_Desktop_Computer_Simple_Loop
root=QualityLevel_0
nodes [
 {
  id=Enveloper_0
  children [Modulator_0]
  properties {
   bLoop=true
   bLoopIndefinitely=true
   loopEnd=20
   pitchCurve {editorCurveData {keys [{time=0}]}}
   volumeCurve {editorCurveData {keys [<at -0.0028 1>,<at 0 1>,<at 10 0.95>,<at 20 1>]}}
  }
 }
 {id=Mixer_0, children [WavePlayer_36], properties {inputVolume [0.5]}}
 {id=Mixer_1, children [WavePlayer_39], properties {inputVolume [0.55]}}
 {id=Mixer_2, children [WavePlayer_34], properties {inputVolume [0.5]}}
 {id=Mixer_3, children [WavePlayer_35], properties {inputVolume [0.7]}}
 {id=Mixer_4, children [WavePlayer_33], properties {inputVolume [0.7]}}
 {id=Mixer_5, children [WavePlayer_37], properties {inputVolume [0.8]}}
 {id=Mixer_6, children [WavePlayer_38], properties {inputVolume [0.8]}}
 {id=Modulator_0, children [Random_0], properties {pitchMax=1.1, pitchMin=0.9, volumeMax=1, volumeMin=0.9}}
 {id=QualityLevel_0, children [Enveloper_0,Enveloper_0,null,null]}
 {id=Random_0, children [Mixer_4,Mixer_2,Mixer_3,Mixer_0,Mixer_5,Mixer_6,Mixer_1], properties {weights [1,1,1,1,1,1,1]}}
 {
  id=WavePlayer_33
  properties {bLooping=true, soundWaveAssetPtr=&/Spot_Vintage_Computer_Idling_Loop_01}
 }
 {
  id=WavePlayer_34
  properties {bLooping=true, soundWaveAssetPtr=&/Spot_Vintage_Computer_Idling_Loop_02}
 }
 {
  id=WavePlayer_35
  properties {bLooping=true, soundWaveAssetPtr=&/Spot_Vintage_Computer_Idling_Loop_03}
 }
 {
  id=WavePlayer_36
  properties {bLooping=true, soundWaveAssetPtr=&/Spot_Vintage_Computer_Idling_Loop_04}
 }
 {
  id=WavePlayer_37
  properties {bLooping=true, soundWaveAssetPtr=&/Spot_Vintage_Computer_Idling_Loop_05}
 }
 {
  id=WavePlayer_38
  properties {bLooping=true, soundWaveAssetPtr=&/Spot_Vintage_Computer_Idling_Loop_06}
 }
 {
  id=WavePlayer_39
  properties {bLooping=true, soundWaveAssetPtr=&/Spot_Vintage_Computer_Idling_Loop_07}
 }
]
properties {
 bDebug=true
 bOverrideAttenuation=true
 bOverrideConcurrency=true
 maxDistance=600
 soundClassObject=&Fort_Ambient_Indoors_Athena
 volumeMultiplier=0.3
 attenuationOverrides {
  bEnableListenerFocus=true
  bEnableOcclusion=true
  bUseComplexCollisionForOcclusion=true
  falloffDistance=500
  focusAzimuth=15
  nonFocusAzimuth=45
  nonFocusVolumeAttenuation=0.4
  occlusionLowPassFilterFrequency=850
  occlusionTraceChannel=ECC_GameTraceChannel11
  spatializationAlgorithm=SPATIALIZATION_HRTF
  stereoSpread=50
  attenuationShapeExtents {x=100}
 }
 concurrencyOverrides {maxCount=3, voiceStealReleaseTime=2}
 concurrencySet [/AmbientAudioBase/Concurrency/AmbientConc_4Max_Spot_Distance_Concurency]
}
