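# Sound cue for an indoor ambient vintage desktop computer spot sound, spatialized with HRTF, occlusion, listener focus and tight concurrency limits.
# It mixes a randomly picked idle loop, pitch-varied and shaped by a looping volume envelope, with four intermittent layers: servo whirrs, disk read/write, short servo sounds and stepper motors.
# Those layers loop through random delays, picks and volumes.
# Every layer connects only to the first two quality-level slots.
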SoundCue /WildEstate/Assets/Audio/POI/Domestic_Commercial/Spots/SC_Spot_Vintage_Desktop_Computer_Verbose_Loop
root=Mixer_1
nodes [
 {id=Delay_0, children [Modulator_1], properties {delayMax=15, delayMin=5}}
 {id=Delay_1, children [Modulator_2], properties {delayMax=15, delayMin=5}}
 {id=Delay_2, children [Modulator_3], properties {delayMax=15, delayMin=5}}
 {id=Delay_3, children [Modulator_4], properties {delayMax=15, delayMin=5}}
 {
  id=Enveloper_0
  children [Modulator_0]
  properties {
   bLoop=true
   bLoopIndefinitely=true
   loopEnd=20
   pitchCurve {editorCurveData {keys [{time=0}]}}
   volumeCurve {editorCurveData {keys [<at -0.0028 1>,<at 0 1>,<at 10 0.95>,<at 20 1>]}}
  }
 }
 {id=Looping_0, children [Delay_0]}
 {id=Looping_1, children [Delay_1]}
 {id=Looping_2, children [Delay_2]}
 {id=Looping_3, children [Delay_3]}
 {id=Mixer_0, children [WavePlayer_3], properties {inputVolume [0.5]}}
 {
  id=Mixer_1
  children [QualityLevel_1, QualityLevel_0, QualityLevel_2, QualityLevel_3, QualityLevel_4]
  properties {inputVolume [0.8,0.75,0.75,0.75,1]}
 }
 {id=Mixer_2, children [WavePlayer_6], properties {inputVolume [0.55]}}
 {id=Mixer_3, children [WavePlayer_1], properties {inputVolume [0.5]}}
 {id=Mixer_4, children [WavePlayer_2], properties {inputVolume [0.7]}}
 {id=Mixer_5, children [WavePlayer_0], properties {inputVolume [0.7]}}
 {id=Mixer_6, children [WavePlayer_4], properties {inputVolume [0.8]}}
 {id=Mixer_7, children [WavePlayer_5], properties {inputVolume [0.8]}}
 {id=Modulator_0, children [Random_3], properties {pitchMax=1.1, pitchMin=0.9, volumeMax=1, volumeMin=0.9}}
 {id=Modulator_1, children [Random_2], properties {volumeMax=1, volumeMin=0.9}}
 {id=Modulator_2, children [Random_4], properties {volumeMax=1, volumeMin=0.9}}
 {id=Modulator_3, children [Random_5], properties {volumeMax=1, volumeMin=0.9}}
 {id=Modulator_4, children [Random_1], properties {volumeMax=1, volumeMin=0.9}}
 {id=QualityLevel_0, children [Looping_0,Looping_0,null,null]}
 {id=QualityLevel_1, children [Enveloper_0,Enveloper_0,null,null]}
 {id=QualityLevel_2, children [Looping_1,Looping_1,null,null]}
 {id=QualityLevel_3, children [Looping_2,Looping_2,null,null]}
 {id=QualityLevel_4, children [Looping_3,Looping_3,null,null]}
 {id=Random_1, children [WavePlayer_47, WavePlayer_48, WavePlayer_49, WavePlayer_50], properties {weights [1,1,1,1]}}
 {
  id=Random_2
  children [WavePlayer_8, WavePlayer_9, WavePlayer_10, WavePlayer_11, WavePlayer_12, WavePlayer_13]
  properties {weights [1,1,1,1,1,1]}
 }
 {id=Random_3, children [Mixer_5,Mixer_3,Mixer_4,Mixer_0,Mixer_6,Mixer_7,Mixer_2], properties {weights [1,1,1,1,1,1,1]}}
 {
  id=Random_4
  children [WavePlayer_17, WavePlayer_18, WavePlayer_19, WavePlayer_20, WavePlayer_21, WavePlayer_22, WavePlayer_23]
  properties {weights [1,1,1,1,1,1,1]}
 }
 {
  id=Random_5
  children [WavePlayer_24, WavePlayer_25, WavePlayer_26, WavePlayer_27, WavePlayer_28, WavePlayer_29, WavePlayer_30]
  properties {weights [1,1,1,1,1,1,1]}
 }
 {
  id=WavePlayer_0
  properties {bLooping=true, soundWaveAssetPtr=&/Spot_Vintage_Computer_Idling_Loop_01}
 }
 {
  id=WavePlayer_1
  properties {bLooping=true, soundWaveAssetPtr=&/Spot_Vintage_Computer_Idling_Loop_02}
 }
 {
  id=WavePlayer_10
  properties {soundWaveAssetPtr=&/Spot_Computer_Servos_Movement_Whirring_03}
 }
 {
  id=WavePlayer_11
  properties {soundWaveAssetPtr=&/Spot_Computer_Servos_Movement_Whirring_04}
 }
 {
  id=WavePlayer_12
  properties {soundWaveAssetPtr=&/Spot_Computer_Servos_Movement_Whirring_05}
 }
 {
  id=WavePlayer_13
  properties {soundWaveAssetPtr=&/Spot_Computer_Servos_Movement_Whirring_06}
 }
 {id=WavePlayer_17, properties {soundWaveAssetPtr=&/Spot_Computer_Disk_Drive_Writing_03}}
 {id=WavePlayer_18, properties {soundWaveAssetPtr=&/Spot_Computer_Disk_Drive_Writing_02}}
 {id=WavePlayer_19, properties {soundWaveAssetPtr=&/Spot_Computer_Disk_Drive_Writing_01}}
 {
  id=WavePlayer_2
  properties {bLooping=true, soundWaveAssetPtr=&/Spot_Vintage_Computer_Idling_Loop_03}
 }
 {id=WavePlayer_20, properties {soundWaveAssetPtr=&/Spot_Computer_Disk_Drive_Reading_04}}
 {id=WavePlayer_21, properties {soundWaveAssetPtr=&/Spot_Computer_Disk_Drive_Reading_03}}
 {id=WavePlayer_22, properties {soundWaveAssetPtr=&/Spot_Computer_Disk_Drive_Reading_02}}
 {id=WavePlayer_23, properties {soundWaveAssetPtr=&/Spot_Computer_Disk_Drive_Reading_01}}
 {
  id=WavePlayer_24
  properties {soundWaveAssetPtr=&/Spot_Computer_Disk_Drive_Short_Servo_Resonance_03}
 }
 {
  id=WavePlayer_25
  properties {soundWaveAssetPtr=&/Spot_Computer_Disk_Drive_Short_Servo_Resonance_02}
 }
 {
  id=WavePlayer_26
  properties {soundWaveAssetPtr=&/Spot_Computer_Disk_Drive_Short_Servo_Resonance_01}
 }
 {
  id=WavePlayer_27
  properties {soundWaveAssetPtr=&/Spot_Computer_Disk_Drive_Short_Servo_Movement_04}
 }
 {
  id=WavePlayer_28
  properties {soundWaveAssetPtr=&/Spot_Computer_Disk_Drive_Short_Servo_Movement_03}
 }
 {
  id=WavePlayer_29
  properties {soundWaveAssetPtr=&/Spot_Computer_Disk_Drive_Short_Servo_Movement_02}
 }
 {
  id=WavePlayer_3
  properties {bLooping=true, soundWaveAssetPtr=&/Spot_Vintage_Computer_Idling_Loop_04}
 }
 {
  id=WavePlayer_30
  properties {soundWaveAssetPtr=&/Spot_Computer_Disk_Drive_Short_Servo_Movement_01}
 }
 {
  id=WavePlayer_4
  properties {bLooping=true, soundWaveAssetPtr=&/Spot_Vintage_Computer_Idling_Loop_05}
 }
 {
  id=WavePlayer_47
  properties {soundWaveAssetPtr=&/Spot_Computer_Drive_Stepping_Motor_Movement_04}
 }
 {
  id=WavePlayer_48
  properties {soundWaveAssetPtr=&/Spot_Computer_Drive_Stepping_Motor_Movement_03}
 }
 {
  id=WavePlayer_49
  properties {soundWaveAssetPtr=&/Spot_Computer_Drive_Stepping_Motor_Movement_02}
 }
 {
  id=WavePlayer_5
  properties {bLooping=true, soundWaveAssetPtr=&/Spot_Vintage_Computer_Idling_Loop_06}
 }
 {
  id=WavePlayer_50
  properties {soundWaveAssetPtr=&/Spot_Computer_Drive_Stepping_Motor_Movement_01}
 }
 {
  id=WavePlayer_6
  properties {bLooping=true, soundWaveAssetPtr=&/Spot_Vintage_Computer_Idling_Loop_07}
 }
 {
  id=WavePlayer_8
  properties {soundWaveAssetPtr=&/Spot_Computer_Servos_Movement_Whirring_01}
 }
 {
  id=WavePlayer_9
  properties {soundWaveAssetPtr=&/Spot_Computer_Servos_Movement_Whirring_02}
 }
]
properties {
 bDebug=true
 bOverrideAttenuation=true
 bOverrideConcurrency=true
 maxDistance=530
 soundClassObject=&Fort_Ambient_Indoors_Athena
 volumeMultiplier=0.3
 attenuationOverrides {
  bEnableListenerFocus=true
  bEnableOcclusion=true
  bUseComplexCollisionForOcclusion=true
  falloffDistance=450
  focusAzimuth=15
  nonFocusAzimuth=45
  nonFocusVolumeAttenuation=0.4
  occlusionLowPassFilterFrequency=850
  occlusionTraceChannel=ECC_GameTraceChannel11
  spatializationAlgorithm=SPATIALIZATION_HRTF
  stereoSpread=0
  attenuationShapeExtents {x=80}
 }
 concurrencyOverrides {maxCount=2, voiceStealReleaseTime=2}
 concurrencySet [/AmbientAudioBase/Concurrency/AmbientConc_4Max_Spot_Distance_Concurency]
}
