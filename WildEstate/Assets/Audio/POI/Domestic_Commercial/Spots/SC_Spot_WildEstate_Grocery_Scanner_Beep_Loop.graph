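# Sound Cue named as a looping grocery scanner beep spot ambience for WildEstate domestic/commercial points of interest.
# Its node graph is empty with no root, so the cue has no sound source to play.
# It retains Athena ambient class routing, reduced volume, very small shared attenuation and a debug flag.
# An enabled concurrency override caps instances at two, taking precedence over the assigned prevent-new concurrency set.
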SoundCue /WildEstate/Assets/Audio/POI/Domestic_Commercial/Spots/SC_Spot_WildEstate_Grocery_Scanner_Beep_Loop
root=null
nodes []
properties {
 attenuationSettings=&/Spot_WildEstate_Domestic_Commercial_VerySmall_Att
 bDebug=true
 bOverrideConcurrency=true
 maxDistance=700
 soundClassObject=&Fort_Ambient_Athena
 volumeMultiplier=0.4
 attenuationOverrides {bSpatialize=false, stereoSpread=1147.9967, attenuationShapeExtents {x=6826.1992}}
 concurrencyOverrides {maxCount=2}
 concurrencySet [/AmbientAudioBase/Concurrency/AmbientConc_2Max_Loop_PreventNew_Default_Concurrency]
}
